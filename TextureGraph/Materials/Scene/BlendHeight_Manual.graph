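# TextureGraph scene material that height-blends two diffuse and displacement texture sets, with the transition set manually by a Lerp parameter.
# A Switch scalar drives If nodes that swap which set is first, and both heights are scaled by Strength before entering HeightLerpWithTwoHeightMaps.
# The resulting height is expanded to RGB and multiplied by the world-space vertex normal to form a displacement vector, alongside a Tess scalar for tessellation.
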Material /TextureGraph/Materials/Scene/BlendHeight_Manual
graph {
 properties {pixelDepthOffsetMode=PDOM_Legacy, usedShadingModels=Default Lit}
 nodes [
  n1 = ScalarParameter() {defaultValue=1, parameterName=Switch, sliderMax=1}
  n2 = TextureSampleParameter2D() {parameterName=Diffuse1, texture=/TextureGraph/Textures/default/default_Albedo}
  n3 = TextureSampleParameter2D() {parameterName=Diffuse2, texture=&default_Albedo}
  n4 = ScalarParameter() {defaultValue=1, parameterName=Switch, sliderMax=1}
  n5 = TextureSampleParameter2D() {parameterName=Height1, samplerType=SAMPLERTYPE_LinearColor, texture=/TextureGraph/Textures/default/default_Displacement}
  n6 = TextureSampleParameter2D() {parameterName=Height2, samplerType=SAMPLERTYPE_LinearColor, texture=&default_Displacement}
  n7 = ScalarParameter() {defaultValue=10, parameterName=Strength}
  n8 = MaterialFunctionCall!(
   Diffuse 1=If(A=@n1, "A > B"=@n2.RGB, "A == B"=@n3.RGB, "A < B"=@n3.RGB) {constB=0.5}
   Diffuse 2=If(A=@n1, "A > B"=@n3.RGB, "A == B"=@n2.RGB, "A < B"=@n2.RGB) {constB=0.5}
   Transistion Phase=ScalarParameter() {defaultValue=0.5, parameterName=Lerp}
   Height Texture 1=Multiply(A=If(A=@n4, "A > B"=@n5.RGB, "A == B"=@n6.RGB, "A < B"=@n6.RGB) {constB=0.5}, B=@n7)
   Height Texture 2=Multiply(A=@n7, B=If(A=@n4, "A > B"=@n6.RGB, "A == B"=@n5.RGB, "A < B"=@n5.RGB) {constB=0.5})
   Contrast
  ) {materialFunction=/Engine/Functions/Engine_MaterialFunctions02/Texturing/HeightLerpWithTwoHeightMaps}
  Multiply(
   A=VertexNormalWS()
   B=MaterialFunctionCall(R=@n8.Resulting Height Map, G=@n8.Resulting Height Map, B=@n8.Resulting Height Map) {materialFunction=&AppendMany}.RGB
  )
  ScalarParameter() {defaultValue=10, parameterName=Tess, sliderMax=10, sliderMin=1}
 ]
}
