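# Translucent unlit TextureGraph scene material that samples a linear-color Source texture parameter, defaulting to an AO texture.
# An sRGB scalar blends between a per-channel sRGB-to-linear conversion and the raw RGB, with the default selecting raw.
# A second blend driven by IsSourceGrayscale takes the same RGB on both inputs, so it changes nothing.
# The graph shows no connection from either blend to a material output.
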Material /TextureGraph/Materials/Scene/DefaultMaterial
graph {
 properties {blendMode=BLEND_Translucent, pixelDepthOffsetMode=PDOM_Legacy, shadingModel=MSM_Unlit, usedShadingModels=Unlit}
 nodes [
  n1 = TextureSampleParameter2D(UVs=TextureCoordinate()) {parameterName=Source, samplerType=SAMPLERTYPE_LinearColor, texture=/TextureGraph/Textures/default/default_AO}
  LinearInterpolate(
   A=MaterialFunctionCall(
    X=MaterialFunctionCall!(sRGB=@n1.R) {materialFunction=/Engine/Functions/Engine_MaterialFunctions01/ImageAdjustment/sRGBToLinear}.Linear
    Y=MaterialFunctionCall(sRGB=@n1.G) {materialFunction=&sRGBToLinear}.Linear
    Z=MaterialFunctionCall(sRGB=@n1.B) {materialFunction=&sRGBToLinear}.Linear
   ) {materialFunction=&MakeFloat3}.Result
   B=@n1.RGB
   Alpha=ScalarParameter() {defaultValue=1, parameterName=sRGB}
  )
  LinearInterpolate(A=@n1.RGB, B=@n1.RGB, Alpha=ScalarParameter() {parameterName=IsSourceGrayscale})
 ]
}
