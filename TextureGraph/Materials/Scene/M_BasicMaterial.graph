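# Basic Default Lit, Nanite-enabled TextureGraph scene material that exposes PBR maps as texture parameters.
# One Tiling scalar scales the UVs shared by the BaseColor, Emissive, Specular, Metallic, Roughness and Normal samples, each with a default texture.
# BaseColor, Emissive and Specular pass through lerps driven by grayscale scalar flags, though both lerp inputs read the same RGB output here.
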Material /TextureGraph/Materials/Scene/M_BasicMaterial
graph {
 properties {bUsedWithNanite=true, pixelDepthOffsetMode=PDOM_Legacy, usedShadingModels=Default Lit}
 nodes [
  n1 = Multiply(A=TextureCoordinate(), B=ScalarParameter() {defaultValue=1, parameterName=Tiling})
  n2 = TextureSampleParameter2D(UVs=@n1) {parameterName=Emissive, texture=&Black}
  LinearInterpolate(A=@n2.RGB, B=@n2.RGB, Alpha=ScalarParameter() {parameterName=IsGrayScaleEmissive})
  n3 = TextureSampleParameter2D(UVs=@n1) {parameterName=BaseColor, texture=/Engine/EngineResources/Gray}
  LinearInterpolate(A=@n3.RGB, B=@n3.RGB, Alpha=ScalarParameter() {parameterName=IsBaseColorGrayScale})
  n4 = TextureSampleParameter2D(UVs=@n1) {parameterName=Specular, texture=/TextureGraph/Textures/default/default_Specular}
  LinearInterpolate(A=@n4.RGB, B=@n4.RGB, Alpha=ScalarParameter() {parameterName=IsSpecularGrayScale})
  TextureSampleParameter2D(UVs=@n1) {parameterName=Metallic, samplerType=SAMPLERTYPE_LinearColor, texture=/TextureGraph/Textures/default/default_Metalness}
  TextureSampleParameter2D(UVs=@n1) {parameterName=Roughness, samplerType=SAMPLERTYPE_LinearColor, texture=/TextureGraph/Textures/default/default_Roughness}
  TextureSampleParameter2D(UVs=@n1) {parameterName=Normal, samplerType=SAMPLERTYPE_Normal, texture=/Engine/EngineMaterials/FlatNormal}
 ]
}
