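# RenderModeSource is an unlit TextureGraph scene material, apparently for a render mode that displays a raw source texture.
# The Source texture parameter is sampled as linear color with UVs scaled by TileX and TileY.
# The sRGB scalar blends from a per-channel sRGB-to-linear conversion at 0 to the unconverted sample at 1, the default.
# A separate GPUTessMultiplier product and a spare texture coordinate feed no other node.
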Material /TextureGraph/Materials/Scene/RenderModeSource
graph {
 properties {pixelDepthOffsetMode=PDOM_Legacy, shadingModel=MSM_Unlit, usedShadingModels=Unlit}
 nodes [
  n1 = ScalarParameter() {defaultValue=1, parameterName=TileX}
  n2 = ScalarParameter() {defaultValue=1, parameterName=TileY}
  n3 = TextureSampleParameter2D(UVs=Multiply(A=TextureCoordinate(), B=AppendVector(A=@n1, B=@n2))) {parameterName=Source, samplerType=SAMPLERTYPE_LinearColor, texture=/TextureGraph/Textures/default/default_Roughness}
  n4 = MaterialFunctionCall!(sRGB=@n3.R) {materialFunction=/Engine/Functions/Engine_MaterialFunctions01/ImageAdjustment/sRGBToLinear}
  n5 = MaterialFunctionCall(sRGB=@n3.G) {materialFunction=&sRGBToLinear}
  n6 = MaterialFunctionCall(sRGB=@n3.B) {materialFunction=&sRGBToLinear}
  LinearInterpolate(
   A=Multiply(A=MaterialFunctionCall(R=@n4.Linear, G=@n5.Linear, B=@n6.Linear) {materialFunction=&AppendMany}.RGB)
   B=@n3.RGB
   Alpha=ScalarParameter() {defaultValue=1, parameterName=sRGB}
  )
  Multiply(A=ScalarParameter() {defaultValue=1, parameterName=GPUTessMultiplier, sliderMax=10}, B=Constant() {r=20})
  TextureCoordinate()
 ]
}
