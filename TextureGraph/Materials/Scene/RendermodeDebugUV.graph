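# Unlit material in the TextureGraph plugin's scene materials, named as a UV debug render mode, with ray-traced shadows disabled.
# Its graph holds a texture coordinate node and a GPUTessMultiplier scalar parameter multiplied by a constant.
# No link from either branch to a material output appears in the graph.
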Material /TextureGraph/Materials/Scene/RendermodeDebugUV
graph {
 properties {bCastRayTracedShadows=false, pixelDepthOffsetMode=PDOM_Legacy, shadingModel=MSM_Unlit, usedShadingModels=Unlit}
 nodes [
  Multiply(A=ScalarParameter() {defaultValue=1, parameterName=GPUTessMultiplier, sliderMax=10}, B=Constant() {r=20})
  TextureCoordinate()
 ]
}
