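# Unlit material in the TextureGraph scene materials, named as a world-tangent debug render mode, though the graph computes no tangents.
# It samples a Source texture parameter at an explicit mip level, converts R, G and B from sRGB to linear separately, and recombines them with AppendMany.
# A separate GPUTessMultiplier parameter is scaled by a constant.
# Neither branch shows a connection to a material output.
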Material /TextureGraph/Materials/Scene/RendermodeDebugWorldTangents
graph {
 properties {bCastRayTracedShadows=false, pixelDepthOffsetMode=PDOM_Legacy, shadingModel=MSM_Unlit, usedShadingModels=Unlit}
 nodes [
  n1 = TextureSampleParameter2D!(UVs=TextureCoordinate(), Level, Apply View MipBias) {mipValueMode=TMVM_MipLevel, parameterName=Source, texture=/TextureGraph/Textures/default/default_Albedo}
  MaterialFunctionCall(
   R=MaterialFunctionCall!(sRGB=@n1.R) {materialFunction=/Engine/Functions/Engine_MaterialFunctions01/ImageAdjustment/sRGBToLinear}.Linear
   G=MaterialFunctionCall(sRGB=@n1.G) {materialFunction=&sRGBToLinear}.Linear
   B=MaterialFunctionCall(sRGB=@n1.B) {materialFunction=&sRGBToLinear}.Linear
  ) {materialFunction=&AppendMany}
  Multiply(A=ScalarParameter() {defaultValue=1, parameterName=GPUTessMultiplier, sliderMax=10}, B=Constant() {r=20})
 ]
}
