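# RendermodeLayerMask is an unlit TextureGraph scene material, apparently for displaying a layer mask as a render mode.
# It samples the green channel of the Source texture parameter using UVs scaled by TileX and TileY, then passes it through sRGBToLinear.
# A separate GPUTessMultiplier parameter is scaled by a constant, with no visible link to the mask chain.
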Material /TextureGraph/Materials/Scene/RendermodeLayerMask
graph {
 properties {
  bCastRayTracedShadows=false
  materialDecalResponse=MDR_None
  pixelDepthOffsetMode=PDOM_Legacy
  shadingModel=MSM_Unlit
  usedShadingModels=Unlit
 }
 nodes [
  n1 = TextureCoordinate()
  n2 = AppendVector(A=ScalarParameter() {defaultValue=1, parameterName=TileX}, B=ScalarParameter() {defaultValue=1, parameterName=TileY})
  MaterialFunctionCall!(sRGB=TextureSampleParameter2D(UVs=Multiply(A=@n1, B=@n2)) {parameterName=Source, texture=/TextureGraph/Textures/default/default_Albedo}.G) {materialFunction=/Engine/Functions/Engine_MaterialFunctions01/ImageAdjustment/sRGBToLinear}
  Multiply(A=ScalarParameter() {defaultValue=1, parameterName=GPUTessMultiplier, sliderMax=10}, B=Constant() {r=20})
 ]
}
