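# Unlit TextureGraph scene material that, by its name, serves as the specular render mode for viewing a material's specular map.
# It samples a Specular texture parameter, defaulting to the albedo placeholder texture, with UVs scaled by TileX and TileY parameters.
# A separate GPUTessMultiplier parameter is scaled by a constant, with no link to the texture sampling.
# Ray-traced shadow casting is disabled.
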Material /TextureGraph/Materials/Scene/RendermodeSpecular
graph {
 properties {bCastRayTracedShadows=false, pixelDepthOffsetMode=PDOM_Legacy, shadingModel=MSM_Unlit, usedShadingModels=Unlit}
 nodes [
  Multiply(A=ScalarParameter() {defaultValue=1, parameterName=GPUTessMultiplier, sliderMax=10}, B=Constant() {r=20})
  n1 = ScalarParameter() {defaultValue=1, parameterName=TileX}
  n2 = ScalarParameter() {defaultValue=1, parameterName=TileY}
  TextureSampleParameter2D(UVs=Multiply(A=TextureCoordinate(), B=AppendVector(A=@n1, B=@n2))) {parameterName=Specular, texture=/TextureGraph/Textures/default/default_Albedo}
 ]
}
