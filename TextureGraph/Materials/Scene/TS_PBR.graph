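# TS_PBR is a Default Lit PBR material in TextureGraph's scene folder, exposing Albedo, Normal, Roughness, Metalness and Displacement texture parameters on shared UVs.
# Displacement is centered on zero, scaled by a multiplier along the vertex normal, then divided by a vector packed from UV channels 1 and 2.
# Three customized UVs and an ObjectScale node suggest that vector carries object scale.
# The normal map is sampled as linear color and unpacked by hand.
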Material /TextureGraph/Materials/Scene/TS_PBR
graph {
 properties {
  numCustomizedUVs=3
  pixelDepthOffsetMode=PDOM_Legacy
  usedShadingModels=Default Lit
  lightmassSettings {
   bCastShadowAsMasked=false
   bOverrideCastShadowAsMasked=false
   bOverrideDiffuseBoost=false
   bOverrideEmissiveBoost=false
   bOverrideExportResolutionScale=false
   diffuseBoost=2
   emissiveBoost=1
   exportResolutionScale=1
  }
 }
 nodes [
  n1 = MaterialFunctionCall() {materialFunction=/Engine/Functions/Engine_MaterialFunctions02/WorldPositionOffset/ObjectScale}
  AppendVector(A=@n1.Scale X, B=@n1.Scale X)
  n2 = TextureCoordinate()
  n3 = TextureSampleParameter2D!(UVs=@n2, Level, Apply View MipBias) {
   automaticViewMipBias=false
   isDefaultMeshpaintTexture=true
   mipValueMode=TMVM_MipLevel
   parameterName=Displacement
   samplerType=SAMPLERTYPE_LinearColor
   texture=null
  }
  n4 = Subtract(A=@n3.R) {constB=0.5}
  n5 = ScalarParameter() {defaultValue=15, parameterName=DisplacmentMultiplyer}
  n6 = TextureCoordinate() {coordinateIndex=1}
  n7 = TextureCoordinate() {coordinateIndex=2}
  Divide(
   A=Multiply(A=VertexNormalWS(), B=Multiply(A=@n4, B=@n5) {constB=5})
   B=MaterialFunctionCall(R=ComponentMask(None=@n6) {r=true}, G=ComponentMask(None=@n6) {g=true}, B=ComponentMask(None=@n7) {r=true}) {materialFunction=&AppendMany}.RGB
  )
  ScalarParameter() {defaultValue=1, parameterName=GPUTessMultiplier, sliderMax=10}
  n8 = TextureSampleParameter2D(UVs=@n2) {mipValueMode=TMVM_MipLevel, parameterName=Normal, samplerType=SAMPLERTYPE_LinearColor, texture=/TextureGraph/Textures/default/default_Normal}
  Subtract(A=Multiply(A=@n8.RGB) {constB=2})
  TextureSampleParameter2D!(UVs=@n2, Apply View MipBias) {parameterName=Albedo, texture=/TextureGraph/Textures/default/default_Albedo}
  TextureSampleParameter2D!(UVs=@n2, Level, Apply View MipBias) {
   mipValueMode=TMVM_MipLevel
   parameterName=Roughness
   samplerType=SAMPLERTYPE_LinearColor
   texture=/TextureGraph/Textures/default/default_Roughness
  }
  TextureSampleParameter2D!(UVs=@n2, Apply View MipBias) {parameterName=Metalness, samplerType=SAMPLERTYPE_LinearColor, texture=/TextureGraph/Textures/default/default_Metalness}
 ]
}
